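; Battle bus actor for the Athena lobby that plays engine audio and music once the game phase step reaches GetReady.
; BeginPlay listens to the stateful phase-step event through the GameplayEventRouter, and EndPlay stops listening.
; GetReady plays an ignition sound, then after a delay starts both loops and records the time.
; Tick maps time until warmup countdown end through a fade curve onto the engine loop's volume, debug-printing the value.
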(event EventEndPlay
  (params (EndPlayReason :type EEndPlayReason%20Enum :default "Destroyed"))
  (|CallParentFunction EndPlayReason)
  (bind _gamephasestepupdatedasynctask (Variables|Default|GetGamePhaseStepUpdatedAsyncTask)
    (:then
      (GameplayEventRouter|StopListeningtoStatefulEvent _gamephasestepupdatedasynctask))
    (:else)))

(event EventBeginPlay
  (|CallParentFunction)
  (bind _returnvalue (GameplayEventRouter|GetEventRouter "/Script/FortniteGame.FortGameState" self))
  (bind (_asynctaskproxy _payload) (GameplayEventRouter|StartListeningToStatefulEvent _returnvalue "/Script/FortniteGame.FortGamePhaseStepUpdatedEvent")
    (:then
      (Variables|Default|SetGamePhaseStepUpdatedAsyncTask _asynctaskproxy))
    (:OnEventReceived
      _)
    (:OnSavedEventStateExists
      _))
  (CallFunction|OnGamePhaseStepChanged (Utilities|Struct|BreakFortGamePhaseStepUpdatedEvent _payload)))

(event Collision|EventActorBeginOverlap
  (params (OtherActor :type Actor%20Object%20Reference)))

(event Custom|On%20Game%20Phase%20Step%20Changed
  (params (GamePhaseStep :type EAthenaGamePhaseStep%20Enum))
  (switch Utilities|FlowControl|Switch|SwitchonEAthenaGamePhaseStep GamePhaseStep
    (:GetReady
      (Audio|SpawnSoundAttached (Variables|Default|GetIgnitionSound) (Variables|Debug|GetStaticMeshComponent))
      (Utilities|FlowControl|Delay (Variables|Default|GetIgnitionTime))
      (Utilities|FlowControl|Sequence
        (:then_0
          (bind _returnvalue (Audio|SpawnSoundAttached (Variables|Default|GetLoopingSound) (Variables|Debug|GetStaticMeshComponent)))
          (Variables|Default|SetLoopingSound_Inst _returnvalue))
        (:then_1
          (bind _returnvalue_2 (Audio|SpawnSoundAttached (Variables|Default|GetLoopingMusic) (Variables|Debug|GetStaticMeshComponent)))
          (Variables|Default|SetLoopingMusic_Inst _returnvalue_2))
        (:then_2
          (Variables|Default|SetGetReadyGameTime (Utilities|Time|GetGameTimeinSeconds)))))))

(event EventTick
  (params (DeltaSeconds :type Float%20%28single-precision%29))
  (bind _returnvalue (BattleRoyaleGamePhaseLogic|GetBattleRoyaleGamePhaseLogicComponent))
  (bind _returnvalue_2 (BattleRoyaleGamePhaseLogic|GetWarmupCountdownEndTime _returnvalue))
  (bind _getreadygametime (Variables|Default|GetGetReadyGameTime))
  (bind _returnvalue_3 (Utilities|Time|GetGameTimeinSeconds))
  (bind _returnvalue_4 (Math|Float|MapRangeClamped _returnvalue_3 _getreadygametime _returnvalue_2 0.0 1))
  (bind _getreadyfadecurve (Variables|Default|GetGetReadyFadeCurve))
  (bind _returnvalue_5 (Math|Curves|GetFloatValue _getreadyfadecurve _returnvalue_4))
  (Utilities|IsValid _returnvalue
    (:Is%20Valid
      (Audio|Components|Audio|SetVolumeMultiplier (Variables|Default|GetLoopingSound_Inst) _returnvalue_5 :!class /Script/Engine.AudioComponent)
      (Development|PrintString (Utilities|String|ToString(Float) _returnvalue_5)))))

(fn ConstructionScript
  (|CallParentFunction))
