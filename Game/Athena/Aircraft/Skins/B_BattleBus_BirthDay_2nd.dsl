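; Battle bus cosmetic Blueprint for the second Birthday bus, adding beat-timed party lights, confetti and audio.
; A BPM-scaled timeline drives three dynamic window-material parameters, shuffling their names each loop, while its beat event counts beats, triggers periodic confetti and prints debug values.
; A game-phase listener starts the active skin's engine and music loops at BusLocked, with a slow timeline controlling music volume.
; EndPlay stops both sounds and the listener.
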(event EventBeginPlay
  (bind _beatcount (Variables|Default|GetBeatCount))
  (bind _materialwindows (Variables|Default|GetMaterialWindows))
  (Utilities|FlowControl|Sequence
    (:then_0
      ;[1 1 1] Audio Control Window Interior lights
      (Components|Timeline|SetPlayRate (Variables|B_BattleBus_BirthDay_2nd|GetLight%20Flash%20Timeline) (* (/ 1.0 (/ 6e+01 (Variables|Default|GetBPM))) 0.25))
      (bind (_direction _light1 _light2 _light3) (Timeline|Light%20Flash%20Timeline PlayFromStart)
        (:Update
          (Utilities|FlowControl|Sequence
            (:then_0
              (Rendering|Material|SetScalarParameterValue _materialwindows (Utilities|Array|Get(aref) (Variables|Default|GetParamNames)) _light1 :!class /Script/Engine.MaterialInstanceDynamic))
            (:then_1
              (Rendering|Material|SetScalarParameterValue _materialwindows (Utilities|Array|Get(aref) (Variables|Default|GetParamNames) 1) _light2 :!class /Script/Engine.MaterialInstanceDynamic))
            (:then_2
              (Rendering|Material|SetScalarParameterValue _materialwindows (Utilities|Array|Get(aref) (Variables|Default|GetParamNames) 2) _light3 :!class /Script/Engine.MaterialInstanceDynamic))))
        (:Loop
          (Utilities|Array|Shuffle (Variables|Default|GetParamNames)))
        (:Beat
          (Utilities|FlowControl|Sequence
            (:then_0
              (if (== (Math|Integer|%25%28Integer%29 _beatcount 2) 0)
                _))
            (:then_1
              (if (== (Math|Integer|%25%28Integer%29 _beatcount 8) 0)
                (Components|Activation|Activate (Variables|Default|GetParticle_Confetti) true)))
            (:then_2
              (bind _result (Math|Integer|IncrementInt (Variables|Default|GetBeatCount))
                (:%20%20
                  ;[0.2969 0.1868 0.1325] Debug
                  (Development|PrintString (Utilities|String|ToString(Integer) _result) true true "(R=0.085431,G=1.000000,B=0.000000,A=1.000000)" 5.0)
                  (Development|PrintString (Utilities|String|ToString(Integer) (Math|Integer|%25%28Integer%29 (Variables|Default|GetBeatCount) 4)) true true "(R=1.000000,G=0.018600,B=0.000000,A=1.000000)" 5.0))))))))
    (:then_1
      (bind _returnvalue (GameplayEventRouter|GetEventRouter "/Script/FortniteGame.FortGameState" self))
      (bind (_asynctaskproxy _payload) (GameplayEventRouter|StartListeningToStatefulEvent _returnvalue "/Script/FortniteGame.FortGamePhaseStepUpdatedEvent")
        (:then
          (Variables|Default|SetGamePhaseStepUpdatedAsyncTask _asynctaskproxy))
        (:OnEventReceived
          _)
        (:OnSavedEventStateExists
          _))
      (CallFunction|GamePhaseStepChanged (Utilities|Struct|BreakFortGamePhaseStepUpdatedEvent _payload)))))

(event Custom|Game%20Phase%20Step%20Changed
  (params (GamePhaseStep :type EAthenaGamePhaseStep%20Enum))
  (switch Utilities|FlowControl|Switch|SwitchonEAthenaGamePhaseStep GamePhaseStep
    (:BusLocked
      (CallFunction|StartAudio))))

(event Custom|GrabAudioReferences
  (bind _activeskin (Variables|BattleBusCosmeticInstanceBase|GetActiveSkin))
  (Variables|Sound|SetLoopingSound (Athena|GetFlightLoopingSound _activeskin))
  (Variables|Sound|SetLoopingMusic (Athena|GetFlightLoopingMusic _activeskin)))

(event Custom|Start%20Audio
  (CallFunction|GrabAudioReferences)
  (Utilities|FlowControl|Sequence
    (:then_0
      (bind _returnvalue (Audio|SpawnSoundAttached :Sound (Variables|Sound|GetLoopingSound) :AttachToComponent (Variables|Default|GetBus) :AttachPointName "None" :LocationType "KeepRelativeOffset" :bStopWhenAttachedToDestroyed true :VolumeMultiplier 1.0 :PitchMultiplier 1.0 :StartTime 0.0 :bAutoDestroy true))
      (Audio|Components|Audio|FadeIn _returnvalue 2 1.0 0.0 "Linear" :!class /Script/Engine.AudioComponent)
      (Variables|Default|SetBusEngine_Inst _returnvalue))
    (:then_1
      (bind _returnvalue_2 (Audio|SpawnSoundAttached :Sound (Variables|Sound|GetLoopingMusic) :AttachToComponent (Variables|Default|GetBus) :AttachPointName "None" :LocationType "KeepRelativeOffset" :bStopWhenAttachedToDestroyed true :VolumeMultiplier 1.0 :PitchMultiplier 1.0 :StartTime 0.0 :bAutoDestroy true))
      (Audio|Components|Audio|FadeIn _returnvalue_2 2 1.0 0.0 "Linear" :!class /Script/Engine.AudioComponent)
      (Variables|Default|SetBusMusic_Inst _returnvalue_2)
      (Components|Timeline|SetPlayRate (Variables|B_BattleBus_BirthDay_2nd|GetBus%20Music%20Fader) (/ 1 30))
      (bind (_direction _volume) (Timeline|Bus%20Music%20Fader PlayFromStart)
        (:Update
          (bind _busmusic_inst (Variables|Default|GetBusMusic_Inst)
            (:then
              (Audio|Components|Audio|SetVolumeMultiplier _busmusic_inst _volume :!class /Script/Engine.AudioComponent))
            (:else)))))))

(event EventEndPlay
  (params (EndPlayReason :type EEndPlayReason%20Enum :default "Destroyed"))
  (Utilities|FlowControl|Sequence
    (:then_0
      (bind _busmusic_inst (Variables|Default|GetBusMusic_Inst)
        (:then
          (Audio|Components|Audio|Stop _busmusic_inst))
        (:else)))
    (:then_1
      (bind _busengine_inst (Variables|Default|GetBusEngine_Inst)
        (:then
          (Audio|Components|Audio|Stop _busengine_inst))
        (:else)))
    (:then_2
      (bind _gamephasestepupdatedasynctask (Variables|Default|GetGamePhaseStepUpdatedAsyncTask)
        (:then
          (GameplayEventRouter|StopListeningtoStatefulEvent _gamephasestepupdatedasynctask))
        (:else)))))

(fn ConstructionScript
  (bind _returnvalue (Rendering|Material|CreateDynamicMaterialInstance (Variables|Default|GetBus) 1 "/Game/Athena/Aircraft/PartyBus_RT/Materials/MI_PartyBus_Anim_Windows.MI_PartyBus_Anim_Windows" "None" :!class /Script/Engine.PrimitiveComponent))
  (Variables|Default|SetMaterialWindows _returnvalue))
