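; Booster Battle Bus skin that plays the cosmetic warp transition from pre-match warm-up into flight.
; Game-state-ready or GetReady starts a warm-up timeline driving camera shake, a lens effect and submix stereo-delay intensity.
; BusLocked pops the bus into the sky: timelines wobble it, fade booster flames, scale and bend the balloon, and drop both canisters, staggered by random delay.
; BusFlying removes the looping lens effect and resets forced input rotation speed.
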(event Custom|StartBusTransitionFX
  ;[1 1 1] Warm Up
  (Networking|CanExecuteCosmeticEvents
    (:True
      (Variables|Default|SetRandomDelayTime (Math|Float|Lerp 0.0 0.5 (Math|Random|RandomFloat)))
      (Development|PrintString "CountDownBeginAndStartToWarmUp" true true "(R=1.000000,G=0.148011,B=0.000000,A=1.000000)" 1e+01)
      (for _array_element (Fortnite|GetLocalFortPlayerControllers true)
        (Utilities|EffectsQualityCheck 0
          (:Cinematic
            _)
          (:Epic
            _)
          (:High
            _)
          (:Medium
            (Variables|Default|SetShake%20Class (Variables|Default|GetShake%20Class_LowEnd)) ; MediumAndLow
            (bind _returnvalue (CameraShakes|StartLegacyCameraShake (Variables|/Script/Engine.PlayerController|GetPlayerCameraManager _array_element) (Variables|Default|GetShake%20Class)))
            (Variables|Default|SetWarmUpCamShake _returnvalue))
          (:Low
            _)) ; FORT-621994
        (Variables|Default|SetShake%20Class (Variables|Default|GetShake%20Class_HighEnd)) ; HighAndAbove
      )
      (bind (_direction _camshake _audio_fx_intensity) (Timeline|Timeline_WarmUP PlayFromStart)
        (:Update
          (CallFunction|UpdateAudioFX _audio_fx_intensity)
          (bind _warmupcamshake (Variables|Default|GetWarmUpCamShake)
            (:then
              (Variables|/Script/Engine.CameraShakeBase|SetShakeScale _camshake _warmupcamshake))
            (:else)))
        (:SpawnWarmUpLens
          (for _array_element_2 (Fortnite|GetLocalFortPlayerControllers true)
            (bind _returnvalue_2 (Feedback|SpawnGenericCameraLensEffect (Game|Player|GetPlayerPawn _array_element_2) "/Game/Environments/Asteria/Props/LobbyIslandSpace/FX/BP_NS_CameraLens_BusWarp_WarmUp.BP_NS_CameraLens_BusWarp_WarmUp_C"))
            (Variables|Default|SetWarmUpLensFX _returnvalue_2))
          (Audio|PlaySoundatLocation "/Game/Athena/Sounds/Vehicles/Bus/Athena_Bus_Warp_Lens_Start_Cue.Athena_Bus_Warp_Lens_Start_Cue" (Transformation|GetWorldLocation (Variables|Default|GetBooster))))
        (:Start%20Audio%20FX
          (Class|SubMixEffectComponent|Enable (Variables|Default|GetSubMixEffectComponent)))))))

(event EventBeginPlay
  (|CallParentFunction)
  (Class|FortAsyncActionGameStateReady|GameStateReadyAsync
    (:OnReady
      (Utilities|FlowControl|Sequence
        (:then_1
          ;[1 1 1] Warm Up
          (Utilities|FlowControl|DoOnce false
            (:Completed
              (Networking|CanExecuteCosmeticEvents
                (:True
                  (Variables|Default|SetRandomDelayTime (Math|Float|Lerp 0.0 0.5 (Math|Random|RandomFloat)))
                  (Development|PrintString "CountDownBeginAndStartToWarmUp" true true "(R=1.000000,G=0.148011,B=0.000000,A=1.000000)" 1e+01)
                  (for _array_element (Fortnite|GetLocalFortPlayerControllers true)
                    (Utilities|EffectsQualityCheck 0
                      (:Cinematic
                        _)
                      (:Epic
                        _)
                      (:High
                        _)
                      (:Medium
                        (Variables|Default|SetShake%20Class (Variables|Default|GetShake%20Class_LowEnd)) ; MediumAndLow
                        (bind _returnvalue (CameraShakes|StartLegacyCameraShake (Variables|/Script/Engine.PlayerController|GetPlayerCameraManager _array_element) (Variables|Default|GetShake%20Class)))
                        (Variables|Default|SetWarmUpCamShake _returnvalue))
                      (:Low
                        _)) ; FORT-621994
                    (Variables|Default|SetShake%20Class (Variables|Default|GetShake%20Class_HighEnd)) ; HighAndAbove
                  )
                  (Timeline|Timeline_WarmUP PlayFromStart))))))))))

(event Custom|StopBusTransitionFX
  (Variables|Default|SetWarmUpCamShake 0)
  (for _array_element (Fortnite|GetLocalFortPlayerControllers true)
    (Camera|RemoveGenericCameraLensEffect (Variables|/Script/Engine.PlayerController|GetPlayerCameraManager _array_element) (Variables|Default|GetWarmUpLensFX)))
  ;[1 1 1] Warm Up
  (Timeline|Timeline_WarmUP Stop))

(event EventGamePhaseStepChanged
  (params (GamePhaseStep :type EAthenaGamePhaseStep%20Enum))
  (bind _warmuplensfx (Variables|Default|GetWarmUpLensFX))
  (bind _fx_battlebus_booster_dislodge_l (Variables|Default|GetFX_BattleBus_Booster_Dislodge_L))
  (bind _fx_battlebus_booster_dislodge_r (Variables|Default|GetFX_BattleBus_Booster_Dislodge_R))
  (bind _balloonflame (Variables|Default|GetBalloonFlame))
  (bind _fx_boosterflames (Variables|Default|GetFX_BoosterFlames))
  (|CallParentFunction GamePhaseStep)
  (switch Utilities|FlowControl|Switch|SwitchonEAthenaGamePhaseStep GamePhaseStep
    (:GetReady
      ;[1 1 1] Warm Up
      (Utilities|FlowControl|DoOnce false
        (:Completed
          (Networking|CanExecuteCosmeticEvents
            (:True
              (Variables|Default|SetRandomDelayTime (Math|Float|Lerp 0.0 0.5 (Math|Random|RandomFloat)))
              (Development|PrintString "CountDownBeginAndStartToWarmUp" true true "(R=1.000000,G=0.148011,B=0.000000,A=1.000000)" 1e+01)
              (for _array_element (Fortnite|GetLocalFortPlayerControllers true)
                (Utilities|EffectsQualityCheck 0
                  (:Cinematic
                    _)
                  (:Epic
                    _)
                  (:High
                    _)
                  (:Medium
                    (Variables|Default|SetShake%20Class (Variables|Default|GetShake%20Class_LowEnd)) ; MediumAndLow
                    (bind _returnvalue (CameraShakes|StartLegacyCameraShake (Variables|/Script/Engine.PlayerController|GetPlayerCameraManager _array_element) (Variables|Default|GetShake%20Class)))
                    (Variables|Default|SetWarmUpCamShake _returnvalue))
                  (:Low
                    _)) ; FORT-621994
                (Variables|Default|SetShake%20Class (Variables|Default|GetShake%20Class_HighEnd)) ; HighAndAbove
              )
              (Timeline|Timeline_WarmUP PlayFromStart))))))
    (:BusLocked
      ;[1 1 1] Pop into Sky
      (for _array_element_2 (Fortnite|GetLocalFortPlayerControllers true)
        (Input|ResetInputRotationScale _array_element_2)
        (bind _returnvalue_2 (Camera|AddGenericCameraLensEffect (Variables|/Script/Engine.PlayerController|GetPlayerCameraManager _array_element_2) "/Game/Environments/Asteria/Props/LobbyIslandSpace/FX/BP_NS_CameraLens_BusWarp_Loop.BP_NS_CameraLens_BusWarp_Loop_C"))
        (Variables|Default|SetBoosterLens _returnvalue_2)
        (CameraShakes|StartLegacyCameraShake (Variables|/Script/Engine.PlayerController|GetPlayerCameraManager _array_element_2) "/Game/Environments/Asteria/Props/LobbyIslandSpace/FX/B_CameraShake_TransitionOut.B_CameraShake_TransitionOut_C"))
      (Components|Activation|Activate (Variables|Default|GetFX_BoosterFlames) false)
      (Rendering|SetVisibility (Variables|Default|GetBalloonFlame) false false)
      (Audio|PlaySoundatLocation "/Game/Athena/Sounds/Vehicles/Bus/Athena_Bus_Warp_Lens_Finish_Cue.Athena_Bus_Warp_Lens_Finish_Cue" (Transformation|GetWorldLocation (Variables|Default|GetBus)))
      (Development|PrintString "PlayTransitionCameraAnimAndFX" true true "(R=0.028321,G=0.000000,B=1.000000,A=1.000000)" 1e+01)
      (Class|SubMixEffectComponent|Disable (Variables|Default|GetSubMixEffectComponent))
      (Utilities|FlowControl|Sequence
        (:then_0
          (Utilities|FlowControl|Delay 0.5)
          (bind _warmupcamshake (Variables|Default|GetWarmUpCamShake)
            (:then
              (Variables|/Script/Engine.CameraShakeBase|SetShakeScale 0.0 _warmupcamshake))
            (:else)))
        (:then_1
          (bind (_direction _cooldown _balloonbend _balloonscaleup _windopacity) (Timeline|Timeline_Cooldown PlayFromStart)
            (:Update
              (Effects|Components|ParticleSystem|SetFloatParameter _fx_boosterflames "OverallOpacity" _cooldown)
              (Effects|Components|ParticleSystem|SetFloatParameter _fx_boosterflames "WindOpacity" _windopacity)
              (Effects|Components|ParticleSystem|SetFloatParameter _fx_boosterflames "BalloonBendAmount" _balloonbend)
              (Effects|Components|ParticleSystem|SetFloatParameter _fx_boosterflames "BalloonSize" _balloonscaleup)
              (CallFunction|UpdateBalloon (Math|Conversions|ToVector(Float) _balloonscaleup) _balloonbend))
            (:Finished
              (Components|Activation|Deactivate _fx_boosterflames))
            (:FireUpMesh
              (Rendering|SetVisibility _balloonflame true false)
              (Audio|PlaySoundatLocation "/Game/Athena/Sounds/Vehicles/Bus/Athena_Bus_Warp_Balloon_Flame_On_Cue.Athena_Bus_Warp_Balloon_Flame_On_Cue" (Transformation|GetWorldLocation _balloonflame)))
            (:FireUpFX
              (Components|Activation|Activate (Variables|Default|GetNS_BattleBus_FireUp) false))
            (:BoosterDropOff
              (Effects|Components|ParticleSystem|SetBoolParameter (Variables|Default|GetFX_BoosterFlames) "KillCanisterWindParticles" true)
              (Rendering|SetVisibility (Variables|Default|GetBattlebus_Booster_Canister_L) false false)
              (Components|Activation|Activate _fx_battlebus_booster_dislodge_l false)
              (Audio|PlaySoundatLocation "/Game/Athena/Sounds/Vehicles/Bus/Athena_Bus_Warp_Thruster_Dislodge_A_Cue.Athena_Bus_Warp_Thruster_Dislodge_A_Cue" (Transformation|GetWorldLocation _fx_battlebus_booster_dislodge_l))
              (Utilities|FlowControl|Delay (Variables|Default|GetRandomDelayTime))
              (Components|Activation|Activate _fx_battlebus_booster_dislodge_r false)
              (Audio|PlaySoundatLocation "/Game/Athena/Sounds/Vehicles/Bus/Athena_Bus_Warp_Thruster_Dislodge_B_Cue.Athena_Bus_Warp_Thruster_Dislodge_B_Cue" (Transformation|GetWorldLocation _fx_battlebus_booster_dislodge_r))
              (Rendering|SetVisibility (Variables|Default|GetBattlebus_Booster_Canister_R) false false))))
        (:then_2
          (Components|Activation|Deactivate (CameraLensEffect|GetPrimaryParticleComponent _warmuplensfx))
          (for _array_element_3 (Fortnite|GetLocalFortPlayerControllers true)
            (Camera|RemoveGenericCameraLensEffect (Variables|/Script/Engine.PlayerController|GetPlayerCameraManager _array_element_3) _warmuplensfx))
          (bind (_direction_2 _newtrack_0) (Timeline|Timeline PlayFromStart)
            (:Update
              (Transformation|SetRelativeRotation (Variables|Default|GetBus) (Math|Rotator|MakeRotator (Math|Float|Lerp -6.0 6.0 (.x _newtrack_0)) (Math|Float|Lerp -6.0 6.0 (.y _newtrack_0)) (+ (Math|Float|Lerp -6.0 6.0 (.z _newtrack_0)) -9e+01)) false false))))))
    (:BusFlying
      ;[1 1 1] Bus Door Open
      (CallFunction|RemoveCameraLensEffect)
      (Development|PrintString "CleanUpScreenFX" true true "(R=1.000000,G=0.000000,B=0.244606,A=1.000000)" 1e+01)
      (for _array_element_4 (Fortnite|GetLocalFortPlayerControllers true)
        (Input|ResetForcedInputRotationSpeed _array_element_4)))))

(event Custom|UpdateBalloon
  (params (Scale3D :type Vector) (BalloonBendAmount :type Float%20%28double-precision%29))
  ;[1 1 1] Pop into Sky
  (Rendering|Material|SetScalarParameterValueonMaterials (Variables|Default|GetBalloon) "BalloonBendAmount" BalloonBendAmount)
  (Transformation|SetRelativeScale3D (Variables|Default|GetBalloon) Scale3D))

(event Custom|RemoveCameraLensEffect
  (bind _boosterlens (Variables|Default|GetBoosterLens))
  (bind _returnvalue (CameraLensEffect|GetPrimaryParticleComponent _boosterlens))
  ;[1 1 1] Bus Door Open
  (if (Utilities|IsValid _returnvalue :!class /Script/Engine.KismetSystemLibrary)
    (Components|Activation|Deactivate _returnvalue))
  (for _array_element (Fortnite|GetLocalFortPlayerControllers true)
    (Camera|RemoveGenericCameraLensEffect (Variables|/Script/Engine.PlayerController|GetPlayerCameraManager _array_element) _boosterlens)))

(event Custom|UnlockCamera)

(fn UpdateAudioFX
  (params (Intensity :type Float%20%28double-precision%29))
  (bind _submixeffectcomponent (Variables|Default|GetSubMixEffectComponent))
  (if (Class|SubMixEffectComponent|GetEnabled _submixeffectcomponent)
    (bind _assubmix_effect_stereo_delay_preset (Utilities|Casting|CastToSubmixEffectStereoDelayPreset (Class|SubMixEffectComponent|GetEffect _submixeffectcomponent)))
    (bind _output_get (LocalVariables|SetSettingsCopy (Variables|/Script/Synthesis.SubmixEffectStereoDelayPreset|GetSettings _assubmix_effect_stereo_delay_preset)))
    (bind _structout (Utilities|Struct|SetmembersinSubmixEffectStereoDelaySettings _output_get (Math|Float|MapRangeClamped (Variables|GetIntensity) 0.0 1.0 2e+03 1.0) (Math|Float|MapRangeClamped (Variables|GetIntensity) 0.0 1.0 0.0 0.6) (Math|Float|MapRangeClamped (- 1.0 (Variables|GetIntensity)) 0.0 1.0 0.0 1.0)))
    (Audio|Effects|Delay|SetSettings _assubmix_effect_stereo_delay_preset _structout :!class /Script/Synthesis.SubmixEffectStereoDelayPreset)))

(fn ConstructionScript
  (Variables|Default|SetWindowMaterialIndex 0)
  (|CallParentFunction))
