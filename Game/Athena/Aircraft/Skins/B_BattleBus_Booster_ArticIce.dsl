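; Battle Bus booster skin Blueprint for an Arctic Ice variant of the aircraft's boosters.
; Its Construction Script calls the parent, then sets the Niagara material variable on the left booster dislodge effect component to the ArcticIceEpoxyWaiter canister material instance.
; No other components or logic are changed in this graph.
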(fn ConstructionScript
  (|CallParentFunction)
  (Niagara|SetNiagaraVariable(Material) (Variables|Default|GetFX_BattleBus_Booster_Dislodge_L) "NewMaterialInterface" "/Game/Environments/Asteria/Props/LobbyIslandSpace/FX/MI_BattleBus_Booster_ArcticIceEpoxyWaiter_Canister.MI_BattleBus_Booster_ArcticIceEpoxyWaiter_Canister"))
