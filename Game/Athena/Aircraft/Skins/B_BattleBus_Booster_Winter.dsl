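; Winter skin variant of the booster Battle Bus in Athena, extending a parent bus Blueprint with festive deco, sound and effects.
; Balloon updates call the parent, then pass scale and a bend material parameter to the winter deco mesh.
; The BusFlying game phase adds a 2D winter bells sound after parent handling.
; The construction script swaps the right booster dislodge Niagara effect's material to a winter canister instance.
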(event EventUpdateBalloon
  (params (Scale3D :type Vector :default "0, 0, 0") (BalloonBendAmount :type Float%20%28double-precision%29))
  (bind _sm_partybus_winter_deco_3 (Variables|Default|GetSM_PartyBus_Winter_Deco_3))
  (|CallParentFunction Scale3D BalloonBendAmount)
  (Rendering|Material|SetScalarParameterValueonMaterials _sm_partybus_winter_deco_3 "BalloonBendAmount" BalloonBendAmount)
  (Transformation|SetRelativeScale3D _sm_partybus_winter_deco_3 Scale3D))

(event Collision|EventActorBeginOverlap
  (params (OtherActor :type Actor%20Object%20Reference)))

(event EventTick
  (params (DeltaSeconds :type Float%20%28single-precision%29)))

(event EventGamePhaseStepChanged
  (params (GamePhaseStep :type EAthenaGamePhaseStep%20Enum))
  (|CallParentFunction GamePhaseStep)
  (switch Utilities|FlowControl|Switch|SwitchonEAthenaGamePhaseStep GamePhaseStep
    (:BusFlying
      ;[0.1374 1 0.3179] Winter Bells
      (Audio|PlaySound2D "/Game/Athena/Sounds/Vehicles/Bus/Athena_Bus_Warp_Lens_Finish_Winter_Bells_Cue.Athena_Bus_Warp_Lens_Finish_Winter_Bells_Cue"))))

(fn ConstructionScript
  (|CallParentFunction)
  (Niagara|SetNiagaraVariable(Material) (Variables|Default|GetFX_BattleBus_Booster_Dislodge_R) "NewMaterialInterface" "/Game/Environments/Asteria/Props/LobbyIslandSpace/FX/MI_BattleBus_Booster_Winter_Canister.MI_BattleBus_Booster_Winter_Canister"))
