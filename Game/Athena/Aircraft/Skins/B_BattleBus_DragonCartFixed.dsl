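; Battle Bus cosmetic skin Blueprint (DragonCartFixed) that animates interior window lights at a BPM-derived rate and plays the bus's flight audio.
; A flash timeline drives three window material parameters, reshuffles their order on loop, and fires beat-count events, which are empty stubs here.
; A game-phase listener starts the active skin's engine and music loops on BusLocked, fading both in while a slow timeline controls music volume.
; EndPlay stops both sounds and the listener.
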(event EventBeginPlay
  (Utilities|FlowControl|Sequence
    (:then_0
      ;[0.2931 0.3926 0.826] Audio Control Window Interior lights
      (Components|Timeline|SetPlayRate (Variables|B_BattleBus_DragonCartFixed|GetLight%20Flash%20Timeline) (SkinnableBus|CalculateWindowLightPlayRate))
      (bind (_direction _light1 _light2 _light3) (Timeline|Light%20Flash%20Timeline PlayFromStart)
        (:Update
          (Protected|SetWindowLightParameters _light1 _light2 _light3))
        (:Loop
          (Protected|RandomizeLightOrder))
        (:Beat
          (Protected|FireBeatResponses))))
    (:then_1
      (bind _returnvalue (GameplayEventRouter|GetEventRouter "/Script/FortniteGame.FortGameState" self))
      (bind (_asynctaskproxy _payload) (GameplayEventRouter|StartListeningToStatefulEvent _returnvalue "/Script/FortniteGame.FortGamePhaseStepUpdatedEvent")
        (:then
          (Variables|Default|SetGamePhaseStepUpdatedAsyncTask _asynctaskproxy))
        (:OnEventReceived
          _)
        (:OnSavedEventStateExists
          _))
      (CallFunction|GamePhaseStepChanged (Utilities|Struct|BreakFortGamePhaseStepUpdatedEvent _payload)))))

(event Custom|Game%20Phase%20Step%20Changed
  (params (GamePhaseStep :type EAthenaGamePhaseStep%20Enum))
  (switch Utilities|FlowControl|Switch|SwitchonEAthenaGamePhaseStep GamePhaseStep
    (:BusLocked
      (SkinnableBus|StartBusAudio))))

(event Custom|OnBeat)

(event Custom|On2Beat)

(event EventEndPlay
  (params (EndPlayReason :type EEndPlayReason%20Enum :default "Destroyed"))
  (SkinnableBus|HandleEndPlay)
  (bind _gamephasestepupdatedasynctask (Variables|Default|GetGamePhaseStepUpdatedAsyncTask)
    (:then
      (GameplayEventRouter|StopListeningtoStatefulEvent _gamephasestepupdatedasynctask))
    (:else)))

(event Custom|On3Beat)

(event Custom|On4Beat)

(event Custom|On5Beat)

(event Custom|Start%20Bus%20Music%20Fader
  (bind (_direction _volume) (Timeline|Bus%20Music%20Fader PlayFromStart)
    (:Update
      (bind _busmusic_inst (Variables|Default|GetBusMusic_Inst)
        (:then
          (Audio|Components|Audio|SetVolumeMultiplier _busmusic_inst _volume :!class /Script/Engine.AudioComponent))
        (:else)))))

(event Custom|On6Beat)

(event Custom|On7Beat)

(event Custom|On8Beat)

(fn CalculateWindowLightPlayRate
  (return (* (/ 1.0 (/ 6e+01 (Variables|Sound|GetBPM))) 0.25)))

(fn FireBeatResponses
  (Utilities|FlowControl|Sequence
    (:then_0
      (CallFunction|OnBeat))
    (:then_1
      (Utilities|BeatCountMatches 2
        (:Matched
          (CallFunction|On2Beat))))
    (:then_2
      (Utilities|BeatCountMatches 3
        (:Matched
          (CallFunction|On3Beat))))
    (:then_3
      (Utilities|BeatCountMatches 4
        (:Matched
          (CallFunction|On4Beat))))
    (:then_4
      (Utilities|BeatCountMatches 5
        (:Matched
          (CallFunction|On5Beat))))
    (:then_5
      (Utilities|BeatCountMatches 6
        (:Matched
          (CallFunction|On6Beat))))
    (:then_6
      (Utilities|BeatCountMatches 7
        (:Matched
          (CallFunction|On7Beat))))
    (:then_7
      (Utilities|BeatCountMatches 8
        (:Matched
          (CallFunction|On8Beat))))
    (:then_8
      (bind _result (Math|Integer|IncrementInt (Variables|Protected|GetBeatCount))))))

(fn GrabAudioReferences
  (bind _activeskin (Variables|BattleBusCosmeticInstanceBase|GetActiveSkin))
  (Variables|Sound|SetLoopingSound (Athena|GetFlightLoopingSound _activeskin))
  (Variables|Sound|SetLoopingMusic (Athena|GetFlightLoopingMusic _activeskin)))

(fn HandleEndPlay
  (Utilities|FlowControl|Sequence
    (:then_0
      (bind _busmusic_inst (Variables|Default|GetBusMusic_Inst)
        (:then
          (Audio|Components|Audio|Stop _busmusic_inst))
        (:else)))
    (:then_1
      (bind _busengine_inst (Variables|Default|GetBusEngine_Inst)
        (:then
          (Audio|Components|Audio|Stop _busengine_inst))
        (:else)))))

(fn RandomizeLightOrder
  (Utilities|Array|Shuffle (Variables|InteriorWindows|GetLightParamNames)))

(fn SetWindowLightParameters
  (params (Light%201 :type Float%20%28double-precision%29) (Light%202 :type Float%20%28double-precision%29) (Light%203 :type Float%20%28double-precision%29))
  (for (_array_element _array_index) (Variables|InteriorWindows|GetLightParamNames)
    (Rendering|Material|SetScalarParameterValue (Variables|InteriorWindows|GetMaterialWindows) _array_element (Utilities|Select :Index _array_index :Option%200 Light%201 :Option%201 Light%202 :Option%202 Light%203) :!class /Script/Engine.MaterialInstanceDynamic)))

(fn StartBusAudio
  (Protected|GrabAudioReferences)
  (Utilities|FlowControl|Sequence
    (:then_0
      (bind _returnvalue (Audio|SpawnSoundAttached :Sound (Variables|Sound|GetLoopingSound) :AttachToComponent (Variables|Default|GetBus) :AttachPointName "None" :LocationType "KeepRelativeOffset" :bStopWhenAttachedToDestroyed true :VolumeMultiplier 1.0 :PitchMultiplier 1.0 :StartTime 0.0 :bAutoDestroy true))
      (Audio|Components|Audio|FadeIn _returnvalue 2 1.0 0.0 "Linear" :!class /Script/Engine.AudioComponent)
      (Variables|Default|SetBusEngine_Inst _returnvalue))
    (:then_1
      (bind _returnvalue_2 (Audio|SpawnSoundAttached :Sound (Variables|Sound|GetLoopingMusic) :AttachToComponent (Variables|Default|GetBus) :AttachPointName "None" :LocationType "KeepRelativeOffset" :bStopWhenAttachedToDestroyed true :VolumeMultiplier 1.0 :PitchMultiplier 1.0 :StartTime 0.0 :bAutoDestroy true))
      (Audio|Components|Audio|FadeIn _returnvalue_2 2 1.0 0.0 "Linear" :!class /Script/Engine.AudioComponent)
      (Variables|Default|SetBusMusic_Inst _returnvalue_2)
      (Components|Timeline|SetPlayRate (Variables|B_BattleBus_DragonCartFixed|GetBus%20Music%20Fader) (/ 1 30))
      (CallFunction|StartBusMusicFader))))

(fn ConstructionScript
  (bind _returnvalue (Rendering|Material|CreateDynamicMaterialInstance (Variables|Default|GetBus) 1 "/Game/Athena/Aircraft/PartyBus_RT/Materials/MI_PartyBus_Anim_Windows.MI_PartyBus_Anim_Windows" "None" :!class /Script/Engine.PrimitiveComponent))
  (Variables|InteriorWindows|SetMaterialWindows _returnvalue))
