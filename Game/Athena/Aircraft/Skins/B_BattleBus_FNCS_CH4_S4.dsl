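; Battle Bus skin Blueprint for the FNCS Chapter 4 Season 4 variant of the Athena aircraft.
; Its construction script sets the window material index to the first slot, then calls the parent construction script.
; All other behavior comes from the parent Battle Bus class.
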(fn ConstructionScript
  (Variables|Default|SetWindowMaterialIndex 0)
  (|CallParentFunction))
