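; Battle bus skin Blueprint, named for a headband event, that derives from a parent class with audio beat response events.
; It overrides beat events from every beat through every eighth beat as hooks for firing effects in time with audio.
; Each override and the construction script only calls the parent, so this graph adds no logic of its own.
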(event EventOnBeat
  ;[0.4034 0.5567 1] Audio Beat Responses: To fire an FX on a certain beat count, just call some blueprint nodes from these events. Each event fires on the 2nd, 3rd, 4th, 5th, 6th, 7th, and 8th beat.
  (|CallParentFunction))

(event EventOn2Beat
  ;[0.4034 0.5567 1] Audio Beat Responses: To fire an FX on a certain beat count, just call some blueprint nodes from these events. Each event fires on the 2nd, 3rd, 4th, 5th, 6th, 7th, and 8th beat.
  (|CallParentFunction))

(event EventOn3Beat
  ;[0.4034 0.5567 1] Audio Beat Responses: To fire an FX on a certain beat count, just call some blueprint nodes from these events. Each event fires on the 2nd, 3rd, 4th, 5th, 6th, 7th, and 8th beat.
  (|CallParentFunction))

(event EventOn4Beat
  ;[0.4034 0.5567 1] Audio Beat Responses: To fire an FX on a certain beat count, just call some blueprint nodes from these events. Each event fires on the 2nd, 3rd, 4th, 5th, 6th, 7th, and 8th beat.
  (|CallParentFunction))

(event EventOn5Beat
  ;[0.4034 0.5567 1] Audio Beat Responses: To fire an FX on a certain beat count, just call some blueprint nodes from these events. Each event fires on the 2nd, 3rd, 4th, 5th, 6th, 7th, and 8th beat.
  (|CallParentFunction))

(event EventOn6Beat
  ;[0.4034 0.5567 1] Audio Beat Responses: To fire an FX on a certain beat count, just call some blueprint nodes from these events. Each event fires on the 2nd, 3rd, 4th, 5th, 6th, 7th, and 8th beat.
  (|CallParentFunction))

(event EventOn7Beat
  ;[0.4034 0.5567 1] Audio Beat Responses: To fire an FX on a certain beat count, just call some blueprint nodes from these events. Each event fires on the 2nd, 3rd, 4th, 5th, 6th, 7th, and 8th beat.
  (|CallParentFunction))

(event EventOn8Beat
  ;[0.4034 0.5567 1] Audio Beat Responses: To fire an FX on a certain beat count, just call some blueprint nodes from these events. Each event fires on the 2nd, 3rd, 4th, 5th, 6th, 7th, and 8th beat.
  (|CallParentFunction))

(fn ConstructionScript
  (|CallParentFunction))
