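; Rainbow Royale skin Blueprint for the Battle Bus aircraft, derived from a parent class that supplies the base construction logic.
; Its construction script runs the parent's, then creates a dynamic material instance from Window Source Material on slot 1 of the Bus component.
; The instance is stored in Material Windows, under the InteriorWindows category.
; The BeginPlay, overlap and Tick events add no logic of their own.
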(event EventBeginPlay
  (|CallParentFunction) ; This node is disabled and will not be called.\nDrag off pins to build functionality.
)

(event Collision|EventActorBeginOverlap
  (params (OtherActor :type Actor%20Object%20Reference)))

(event EventTick
  (params (DeltaSeconds :type Float%20%28single-precision%29)))

(fn ConstructionScript
  (|CallParentFunction)
  (bind _returnvalue (Rendering|Material|CreateDynamicMaterialInstance (Variables|Default|GetBus) 1 (Variables|Default|GetWindow%20Source%20Material) "None" :!class /Script/Engine.PrimitiveComponent))
  (Variables|InteriorWindows|SetMaterialWindows _returnvalue))
