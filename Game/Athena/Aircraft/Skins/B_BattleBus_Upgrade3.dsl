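; Battle Bus skin actor (Upgrade 3) that drives the bus's window lights, storm effects and flight audio.
; A music-BPM-paced timeline animates window material parameters, shuffling their names on each loop.
; With new bus FX enabled, that update also writes flight direction to the storm-cloud parameter collection and toggles camera shake and an attached B_BusStorm by safe-zone, cloud-bias and camera-distance checks.
; Active-skin engine and music loops fade in, with timeline-driven music volume.
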(event EventBeginPlay
  (bind _returnvalue (SafeZone|GetSafeZoneIndicator))
  (bind _returnvalue_2 (Game|GetPlayerController 0))
  (bind _busstormfx (Variables|Default|GetBusStormFX))
  (bind _materialwindows (Variables|Default|GetMaterialWindows))
  (Utilities|FlowControl|Sequence
    (:then_0
      ;[1 1 1] Audio Control Window Interior lights
      (Components|Timeline|SetPlayRate (Variables|B_BattleBus_Upgrade3|GetLight%20Flash%20Timeline) (* (/ 1.0 (/ 6e+01 (Variables|Default|GetMusicBPM))) 0.25))
      (bind (_direction _light1 _light2 _light3) (Timeline|Light%20Flash%20Timeline PlayFromStart)
        (:Update
          (Utilities|FlowControl|Sequence
            (:then_0
              (Rendering|Material|SetScalarParameterValue _materialwindows (Utilities|Array|Get(aref) (Variables|Default|GetParamNames)) _light1 :!class /Script/Engine.MaterialInstanceDynamic))
            (:then_1
              (Rendering|Material|SetScalarParameterValue _materialwindows (Utilities|Array|Get(aref) (Variables|Default|GetParamNames) 1) _light2 :!class /Script/Engine.MaterialInstanceDynamic))
            (:then_2
              (Rendering|Material|SetScalarParameterValue _materialwindows (Utilities|Array|Get(aref) (Variables|Default|GetParamNames) 2) _light3 :!class /Script/Engine.MaterialInstanceDynamic))
            (:then_3
              (if (Variables|Default|GetEnableNewBusFX)
                ;[1 1 1] Set the Direction of Cloud Post process material movement
                (Rendering|Material|SetScalarParameterValue "/Game/Athena/SafeZone/MPC_SafeZoneClouds.MPC_SafeZoneClouds" "FlightDirectionAngle" (/ (Fortnite|VectortoNormalizedAngleinDegrees (Math|Vector|GetRightVector (Transformation|GetWorldRotation (Variables|Default|GetBus)))) -3.6e+02) :!class /Script/Engine.KismetMaterialLibrary)
                (Utilities|Casting|CastToFortGameStateAthena (Game|GetGameState))
                (Utilities|IsValid _returnvalue
                  (:Is%20Valid
                    ;[1 1 1] Camera shake while in Storm
                    (bind _returnvalue_3 (Rendering|Material|GetScalarParameterValue "/Game/Athena/SafeZone/MPC_SafeZoneClouds.MPC_SafeZoneClouds" "CloudStartingBias" :!class /Script/Engine.KismetMaterialLibrary))
                    (bind _distance (Class|MissionBlueprintFunctionLibrary|DistanceBetweenTwoVectors (SafeZone|GetSafeZoneCenter _returnvalue) (* (Transformation|GetActorLocation) "1.000000,1.000000, 0")))
                    (bind _distance_2 (Class|MissionBlueprintFunctionLibrary|DistanceBetweenTwoVectors (Camera|GetCameraLocation (Variables|/Script/Engine.PlayerController|GetPlayerCameraManager _returnvalue_2)) (Transformation|GetActorLocation)))
                    (if (and (and (>= _distance (SafeZone|GetSafeZoneRadius _returnvalue)) (< _returnvalue_3 0.5)) (< _distance_2 1e+04))
                      (Game|Feedback|ClientStartCameraShake :self _returnvalue_2 :Shake "/Game/Athena/SafeZone/FX/AmbientLightningCameraShake_bus.AmbientLightningCameraShake_bus_C" :Scale 0.2 :PlaySpace "CameraLocal")
                      (Class|BBusStorm|StartBusStorm _busstormfx)
                      (else
                        (Game|Feedback|ClientStopCameraShake _returnvalue_2 "/Game/Athena/SafeZone/FX/AmbientLightningCameraShake_bus.AmbientLightningCameraShake_bus_C" false)
                        (Class|BBusStorm|EndBusStorm _busstormfx)))))))))
        (:Loop
          (Utilities|Array|Shuffle (Variables|Default|GetParamNames)))))
    (:then_1
      (bind _returnvalue_4 (Game|SpawnActorfromClass "/Game/Athena/Aircraft/Skins/B_BusStorm.B_BusStorm_C" (Transformation|GetActorTransform) "Undefined" "OverrideRootScale" 0 0))
      (Transformation|AttachActorToActor _returnvalue_4 self "None" "SnapToTarget" "SnapToTarget" "KeepRelative" true)
      (Variables|Default|SetBusStormFX _returnvalue_4))
    (:then_2
      (CallFunction|StartAudio))))

(event Custom|GrabAudioReferences
  (bind _activeskin (Variables|BattleBusCosmeticInstanceBase|GetActiveSkin))
  (Variables|Sound|SetLoopingSound (Athena|GetFlightLoopingSound _activeskin))
  (Variables|Sound|SetLoopingMusic (Athena|GetFlightLoopingMusic _activeskin)))

(event Custom|Start%20Audio
  (CallFunction|GrabAudioReferences)
  (Utilities|FlowControl|Sequence
    (:then_0
      (bind _returnvalue (Audio|SpawnSoundAttached :Sound (Variables|Sound|GetLoopingSound) :AttachToComponent (Variables|Default|GetBus) :AttachPointName "None" :LocationType "KeepRelativeOffset" :bStopWhenAttachedToDestroyed true :VolumeMultiplier 1.0 :PitchMultiplier 1.0 :StartTime 0.0 :bAutoDestroy true))
      (Audio|Components|Audio|FadeIn _returnvalue 2 1.0 0.0 "Linear" :!class /Script/Engine.AudioComponent)
      (Variables|Default|SetBusEngine_Inst _returnvalue))
    (:then_1
      (bind _returnvalue_2 (Audio|SpawnSoundAttached :Sound (Variables|Sound|GetLoopingMusic) :AttachToComponent (Variables|Default|GetBus) :AttachPointName "None" :LocationType "KeepRelativeOffset" :bStopWhenAttachedToDestroyed true :VolumeMultiplier 1.0 :PitchMultiplier 1.0 :StartTime 0.0 :bAutoDestroy true))
      (Audio|Components|Audio|FadeIn _returnvalue_2 2 1.0 0.0 "Linear" :!class /Script/Engine.AudioComponent)
      (Variables|Default|SetBusMusic_Inst _returnvalue_2)
      (Components|Timeline|SetPlayRate (Variables|B_BattleBus_Upgrade3|GetBus%20Music%20Fader) (/ 1 30))
      (bind (_direction _volume) (Timeline|Bus%20Music%20Fader PlayFromStart)
        (:Update
          (bind _busmusic_inst (Variables|Default|GetBusMusic_Inst)
            (:then
              (Audio|Components|Audio|SetVolumeMultiplier _busmusic_inst _volume :!class /Script/Engine.AudioComponent))
            (:else)))))))

(event EventEndPlay
  (params (EndPlayReason :type EEndPlayReason%20Enum :default "Destroyed"))
  (Utilities|FlowControl|Sequence
    (:then_0
      (bind _busmusic_inst (Variables|Default|GetBusMusic_Inst)
        (:then
          (Audio|Components|Audio|Stop _busmusic_inst))
        (:else)))
    (:then_1
      (bind _busengine_inst (Variables|Default|GetBusEngine_Inst)
        (:then
          (Audio|Components|Audio|Stop _busengine_inst))
        (:else)))))

(fn ConstructionScript
  (bind _returnvalue (Rendering|Material|CreateDynamicMaterialInstance (Variables|Default|GetBus) 5 0 "None" :!class /Script/Engine.PrimitiveComponent))
  (Variables|Default|SetMaterialWindows _returnvalue)
  (Rendering|Material|SetScalarParameterValue "/Game/Athena/SafeZone/MPC_SafeZoneClouds.MPC_SafeZoneClouds" "FlightDirectionAngle" (/ (Fortnite|VectortoNormalizedAngleinDegrees (Math|Vector|GetRightVector (Transformation|GetWorldRotation (Variables|Default|GetBus)))) -3.6e+02) :!class /Script/Engine.KismetMaterialLibrary))
