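; B_BusStorm is a cosmetic storm effect for a Battle Bus skin, toggled by start and end events.
; Start adds a camera raindrop lens effect, rain particles and a looping timer whose interval is randomized once; end undoes all three.
; Each tick picks a point around the actor, randomly spawns a pooled lightning bolt with sound, and flashes a lazily created, reused point light through a randomized-speed timeline.
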(event EventOnStartBusStorm
  (CallFunction|StartBusStorm))

(event Custom|Start%20Bus%20Storm
  (Networking|CanExecuteCosmeticEvents
    (:True
      (if (not (Utilities|Time|IsTimerActivebyHandle (stale _returnvalue_2)))
        (Utilities|FlowControl|Sequence
          (:then_0
            (bind _returnvalue (Camera|AddGenericCameraLensEffect (Game|GetPlayerCameraManager 0) "/Game/VisualThreatManager/Rifts/Blueprints/B_CameraRainDrops_01.B_CameraRainDrops_01_C"))
            (Variables|Default|SetCameraLensEffectInterface _returnvalue)
            (bind _returnvalue_2 (Utilities|Time|SetTimerbyEvent (delegate Custom|StartRandomLightningFlashes) (Math|Random|RandomFloatinRange (Variables|Default|GetLightningFlashRepeatDelayMin) (Variables|Default|GetLightningFlashRepeatDelayMax)) true)))
          (:then_1
            (Components|Activation|Activate (Variables|Default|GetP_Storm_Rain_Athena_Bus) false)))))))

(event EventOnEndBusStorm
  (CallFunction|EndBusStorm))

(event Custom|EndBusStorm
  (bind _cameralenseffectinterface (Variables|Default|GetCameraLensEffectInterface))
  (Components|Activation|Deactivate (Variables|Default|GetP_Storm_Rain_Athena_Bus))
  (Utilities|FlowControl|Sequence
    (:then_0
      (LensEffect|IsValidCameraLens _cameralenseffectinterface
        (:Valid
          (Camera|RemoveGenericCameraLensEffect (Game|GetPlayerCameraManager 0) _cameralenseffectinterface))))
    (:then_1
      (if (Utilities|Time|IsValidTimerHandle (stale Custom|Start%20Bus%20Storm _returnvalue_2))
        (Utilities|Time|ClearandInvalidateTimerbyHandle (stale Custom|Start%20Bus%20Storm _returnvalue_2))))))

(event Custom|StartRandomLightningFlashes
  (bind _lightningspawnlocation (Variables|Default|GetLightningSpawnLocation))
  ;[1 1 1] Random flashes of light - maybe a mesh of lightning too
  (Variables|Default|SetLightningSpawnLocation (+ (Transformation|GetActorLocation) (CallFunction|GetLocationInCircleXY)))
  (if (> (Math|Random|RandomFloat) (Variables|Default|Get%25ChanceOfLightningMesh))
    (CallFunction|PlayLightningFlashAudio _lightningspawnlocation)
    (bind _returnvalue (FortFX|SpawnPooledFXActor "/Game/Effects/Fort_Effects/World/Enviroment/Lightning/LightningLSystem.LightningLsystem_C" (Math|Transform|MakeTransform _lightningspawnlocation)))
    (bind _aslightning_lsystem (Utilities|Casting|CastToLightningLsystem _returnvalue)
      (:then
        (Rendering|SetTranslucentSortPriority (Variables|/Game/Effects/Fort_Effects/World/Enviroment/Lightning/LightningLSystem.LightningLsystem_C|GetBase _aslightning_lsystem) 2)
        (Rendering|SetTranslucentSortPriority (Variables|/Game/Effects/Fort_Effects/World/Enviroment/Lightning/LightningLSystem.LightningLsystem_C|GetSm_LightningStrikes_B_01 _aslightning_lsystem) 2)
        (Rendering|SetTranslucentSortPriority (Variables|/Game/Effects/Fort_Effects/World/Enviroment/Lightning/LightningLSystem.LightningLsystem_C|GetSm_LightningStrikes_A_01 _aslightning_lsystem) 2)
        (Rendering|SetTranslucentSortPriority (Variables|/Game/Effects/Fort_Effects/World/Enviroment/Lightning/LightningLSystem.LightningLsystem_C|GetSm_LightningStrikes_C_01 _aslightning_lsystem) 2)
        (Class|LightningLsystem|SetupLightning _aslightning_lsystem (+ (Variables|Default|GetLightningSpawnLocation) "0, 0,18000.000000") (Variables|Default|GetLightningSpawnLocation)) ; Start/end is weird because of how old L system was setup
      )
      (:CastFailed
        _))
    (else
      _))
  (Variables|Default|GetRandomLightningLight
    (:then
      _)
    (:else
      (bind _returnvalue_2 (AddComponent|NODE_AddPointLightComponent-0 :bManualAttachment true))
      (Variables|Default|SetRandomLightningLight _returnvalue_2)))
  (Variables|Default|SetLightningIntensity (Math|Random|RandomFloatinRange (Variables|Default|GetLightningIntensityMin) (Variables|Default|GetLightningIntensityMax)))
  (Rendering|SetVisibility (Variables|Default|GetRandomLightningLight) true false)
  (Components|Timeline|SetPlayRate (Variables|B_BusStorm|GetLightinging%20Flash) (Utilities|Reciprocal (Math|Random|RandomFloatinRange (Variables|Default|GetLightningTimelinePlaySpeedMin) (Variables|Default|GetLightningTimelinePlaySpeedMax))))
  (Transformation|SetWorldLocation (Variables|Default|GetRandomLightningLight) (Variables|Default|GetLightningSpawnLocation) false false)
  (bind (_direction _lerp) (Timeline|Lightinging%20Flash PlayFromStart)
    (:Update
      (Rendering|Components|Light|SetIntensity (Variables|Default|GetRandomLightningLight) (Math|Float|Lerp 0.1 (Variables|Default|GetLightningIntensity) _lerp)))
    (:Finished
      (Rendering|SetVisibility (Variables|Default|GetRandomLightningLight) false false))))

(event Custom|Play%20Lightning%20Flash%20Audio
  (params (Location :type Vector))
  (bind (_returnvalue_x _returnvalue_y _returnvalue_z) (Math|Vector|BreakVector (Transformation|GetActorLocation)))
  ;[1 1 1] Random flashes of light - maybe a mesh of lightning too
  (Audio|PlaySoundatLocation "/Game/Athena/Sounds/Shield/SafeZone_Storm_LightningFlash_Cue.SafeZone_Storm_LightningFlash_Cue" (Math|Vector|MakeVector (.x Location) (.y Location) _returnvalue_z)))

(fn GetLocationinCircleXY
  (return (Math|Vector|Lerp(Vector) (Math|Vector|MakeVector 0.0 0.0 (Variables|Default|GetLightningFlashHeight)) (Math|Vector|MakeVector (* (Math|Trig|Sin(Radians) (Math|Float|Lerp 0.0 (* (Math|Trig|GetPI) 2) (Math|Random|RandomFloat))) (Variables|Default|GetLightningFlashDiameter)) (* (Math|Trig|Cos(Radians) (Math|Float|Lerp 0.0 (* (Math|Trig|GetPI) 2) (Math|Random|RandomFloat))) (Variables|Default|GetLightningFlashDiameter)) (Variables|Default|GetLightningFlashHeight)) (Math|Random|RandomFloatinRange 0.25 1))))

(fn ConstructionScript)
