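; Stock Battle Bus cosmetic actor that handles the bus's storm effect, looping audio, window material and rift-out effect.
; BeginPlay attaches a storm FX actor and listens for game phase events, starting engine and music loops at BusLocked and revalidating them at BusFlying.
; A timeline drives music volume, while construction creates the window material and writes flight direction to the safe-zone cloud parameter collection.
; EndPlay stops audio and listening, then spawns rift-out Niagara FX.
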(event EventBeginPlay
  (|CallParentFunction)
  (Utilities|FlowControl|Sequence
    (:then_0
      (bind _returnvalue (Game|SpawnActorfromClass "/Game/Athena/Aircraft/Skins/B_BusStorm.B_BusStorm_C" (Transformation|GetActorTransform) "Undefined" "OverrideRootScale" 0 0))
      (Transformation|AttachActorToActor _returnvalue self "None" "SnapToTarget" "SnapToTarget" "KeepRelative" true)
      (bind _output_get (Variables|Default|SetBusStormFX _returnvalue))
      (CallFunction|SetAthenaBusStorm _output_get))
    (:then_1
      (Class|FortAsyncActionGameStateReady|GameStateReadyAsync
        (:OnReady
          (bind _returnvalue_2 (GameplayEventRouter|GetEventRouter "/Script/FortniteGame.FortGameState" self))
          (bind (_asynctaskproxy _payload) (GameplayEventRouter|StartListeningToStatefulEvent _returnvalue_2 "/Script/FortniteGame.FortGamePhaseStepUpdatedEvent")
            (:then
              (Variables|Default|SetGamePhaseStepUpdatedAsyncTask _asynctaskproxy))
            (:OnEventReceived
              _)
            (:OnSavedEventStateExists
              _))
          (CallFunction|GamePhaseStepChanged (Utilities|Struct|BreakFortGamePhaseStepUpdatedEvent _payload)))))))

(event Custom|Game%20Phase%20Step%20Changed
  (params (GamePhaseStep :type EAthenaGamePhaseStep%20Enum))
  (switch Utilities|FlowControl|Switch|SwitchonEAthenaGamePhaseStep GamePhaseStep
    (:BusLocked
      (CallFunction|StartAudio))
    (:BusFlying
      (CallFunction|ValidateAudio))))

(event Custom|GrabAudioReferences
  (bind _activeskin (Variables|BattleBusCosmeticInstanceBase|GetActiveSkin))
  (Variables|Sound|SetLoopingSound (Athena|GetFlightLoopingSound _activeskin))
  (Variables|Sound|SetLoopingMusic (Athena|GetFlightLoopingMusic _activeskin)))

(event Custom|Start%20Audio
  (CallFunction|GrabAudioReferences)
  (Utilities|FlowControl|Sequence
    (:then_0
      (bind _returnvalue (Audio|SpawnSoundAttached :Sound (Variables|Sound|GetLoopingSound) :AttachToComponent (Variables|Default|GetBus) :AttachPointName "None" :LocationType "KeepRelativeOffset" :bStopWhenAttachedToDestroyed true :VolumeMultiplier 1.0 :PitchMultiplier 1.0 :StartTime 0.0 :bAutoDestroy true))
      (Audio|Components|Audio|FadeIn _returnvalue 0.25 1.0 0.0 "Linear" :!class /Script/Engine.AudioComponent)
      (Variables|Default|SetBusEngine_Inst _returnvalue))
    (:then_1
      (bind _returnvalue_2 (Audio|SpawnSoundAttached :Sound (Variables|Sound|GetLoopingMusic) :AttachToComponent (Variables|Default|GetBus) :AttachPointName "None" :LocationType "KeepRelativeOffset" :bStopWhenAttachedToDestroyed true :VolumeMultiplier 1.0 :PitchMultiplier 1.0 :StartTime 0.0 :bAutoDestroy true))
      (Audio|Components|Audio|FadeIn _returnvalue_2 0.25 1.0 0.0 "Linear" :!class /Script/Engine.AudioComponent)
      (Variables|Default|SetBusMusic_Inst _returnvalue_2)
      (Components|Timeline|SetPlayRate (Variables|B_StockBattleBus|GetBus%20Music%20Fader) (/ 1 30))
      (bind (_direction _volume) (Timeline|Bus%20Music%20Fader PlayFromStart)
        (:Update
          (bind _busmusic_inst (Variables|Default|GetBusMusic_Inst)
            (:then
              (Audio|Components|Audio|SetVolumeMultiplier _busmusic_inst _volume :!class /Script/Engine.AudioComponent))
            (:else)))))))

(event Custom|ValidateAudio
  ;[1 1 1] Checks if sfx/music are playing, if not re triggers bus loop audio for the flying phase
  (Variables|Default|GetAudioShouldPlay
    (:then
      (bind _busengine_inst (Variables|Default|GetBusEngine_Inst)
        (:then
          (bind _busmusic_inst (Variables|Default|GetBusMusic_Inst)
            (:then
              (if (and (Audio|Components|Audio|IsPlaying _busmusic_inst) (Audio|Components|Audio|IsPlaying _busengine_inst))
                _
                (else
                  _)))
            (:else
              _))
          (CallFunction|StartAudio))
        (:else
          _)))
    (:else)))

(event EventEndPlay
  (params (EndPlayReason :type EEndPlayReason%20Enum :default "Destroyed"))
  (Utilities|FlowControl|Sequence
    (:then_0
      (bind _busmusic_inst (Variables|Default|GetBusMusic_Inst)
        (:then
          (Audio|Components|Audio|Stop _busmusic_inst))
        (:else)))
    (:then_1
      (bind _busengine_inst (Variables|Default|GetBusEngine_Inst)
        (:then
          (Audio|Components|Audio|Stop _busengine_inst))
        (:else)))
    (:then_2
      (bind _gamephasestepupdatedasynctask (Variables|Default|GetGamePhaseStepUpdatedAsyncTask)
        (:then
          (GameplayEventRouter|StopListeningtoStatefulEvent _gamephasestepupdatedasynctask))
        (:else)))
    (:then_3
      ; Play rift out FX when this cosmetic actor is destroyed
      (bind _riftoutfx (Variables|Default|GetRiftOutFX)
        (:then
          (Niagara|SpawnSystematLocation _riftoutfx (.location (Transformation|GetActorTransform)) (.rotation (Transformation|GetActorTransform)) (.scale (Transformation|GetActorTransform))))
        (:else)))))

(fn GetBatleBusTimelinePlayRate
  (return (* (/ 1.0 (/ 6e+01 (Variables|Default|GetMusicBPM))) 0.25)))

(fn ConstructionScript
  (bind _returnvalue (Rendering|Material|CreateDynamicMaterialInstance (Variables|Default|GetBus) (Variables|Default|GetWindowMaterialIndex) "/Game/Athena/Aircraft/PartyBus_RT/Materials/MI_PartyBus_Anim_Windows.MI_PartyBus_Anim_Windows" "None" :!class /Script/Engine.PrimitiveComponent))
  (Variables|Effects|SetBattleBusWindowsMaterial _returnvalue)
  (Rendering|Material|SetScalarParameterValue "/Game/Athena/SafeZone/MPC_SafeZoneClouds.MPC_SafeZoneClouds" "FlightDirectionAngle" (/ (Fortnite|VectortoNormalizedAngleinDegrees (Math|Vector|GetRightVector (Transformation|GetWorldRotation (Variables|Default|GetBus)))) -3.6e+02) :!class /Script/Engine.KismetMaterialLibrary))
